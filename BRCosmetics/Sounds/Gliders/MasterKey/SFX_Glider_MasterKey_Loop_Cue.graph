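# Sound cue for the looping flight audio of the Master Key glider cosmetic.
# A Random node picks with equal weight between two looping wave players, each holding one of the two Master Key loop recordings.
# It uses the player glider thrust attenuation, the cosmetic glider loop audio properties sheet and a glider sound class, with slightly reduced volume.
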SoundCue /BRCosmetics/Sounds/Gliders/MasterKey/SFX_Glider_MasterKey_Loop_Cue
root=Random_0
nodes [
 {id=Random_0, children [WavePlayer_1,WavePlayer_2], properties {weights [1,1]}}
 {id=WavePlayer_1, properties {bLooping=true, soundWaveAssetPtr=&/Glider_Master_Key_Loop_01}}
 {id=WavePlayer_2, properties {bLooping=true, soundWaveAssetPtr=&/Glider_Master_Key_Loop_02}}
]
properties {
 attenuationSettings=&Fort_Player_Glider_Thrust_Athena_Att
 audioPropertiesSheet=&APS_Cosmetic_Glider_Loop
 maxDistance=12500
 soundClassObject=&Fort_Gliders_Land
 volumeMultiplier=0.7
}
